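# Physics asset for the female medium Eternity Crown head cosmetic mesh.
# It defines a single body on the faceAttach bone, made of one offset and rotated capsule that approximates part of the head.
# It has no constraints, so the body is not jointed to any other body within this asset.
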PhysicsAsset /BRCosmetics/Characters/Player/Female/Medium/Heads/F_MED_Eternity_Crown_Head/Meshes/Female_Medium_Eternity_Crown_Head_PhysicsAsset
bodies {faceAttach [{shapeType=Capsule, length=7.0919, radius=6.1827, center=<2.1719 7.3719 2.0778>, rotation=<rot 0 100.3486 -46.0906>}]}
constraints []
